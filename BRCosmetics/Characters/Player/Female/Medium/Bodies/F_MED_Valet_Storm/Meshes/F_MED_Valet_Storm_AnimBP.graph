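# Animation Blueprint for the F_MED_Valet_Storm female body cosmetic, which follows the attached parent mesh's pose with a crouch correction.
# Copy Pose From Mesh feeds a component-space Modify Bone node that additively offsets the translation of dyn_glute_l, the only bone adjusted.
# The bIsCrouching flag drives that node's alpha, so the offset applies while crouching.
# The event graph holds only an unconnected update event.
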AnimBlueprint /BRCosmetics/Characters/Player/Female/Medium/Bodies/F_MED_Valet_Storm/Meshes/F_MED_Valet_Storm_AnimBP
target_skeleton=&/F_MED_Valet_Storm_Skeleton
graphs [
 AnimGraph {nodes [
   n1 = "Transform (Modify) Bone - Bone: dyn_glute_l"(
    ComponentPose=Local To Component(LocalPose=Copy Pose From Mesh() {bUseAttachedParent=true})
    Alpha="To Float (Boolean)"(InBool=Get bIsCrouching())
    Translation="0,-3.5,-2.5"
    Rotation="0,0,0"
    Scale="1,1,1"
   ) {translationMode=BMM_Additive, alphaBoolBlend {blendOption=Linear}, boneToModify {boneName=dyn_glute_l}, translation {y=-3.5, z=-2.5}}
   Output Pose(Result=Component To Local(ComponentPose=@n1)) {layerGroup=None}
  ]}
 EventGraph {nodes [Event BlueprintUpdateAnimation()]}
]
variables []
